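# Post-body AnimBlueprint for the CrumbViolin large male character body, layering secondary bone fixes onto the attached parent's pose.
# Copied pose runs through a Control Rig, then shoulder pads get weighted clavicle/upperarm constraints and additive offsets.
# DefaultSlot blends onto both joystick bones during the custom emote montage, or via a front-end idle curve, before constraining them to spine_01.
# Parachute and backbling alphas update but are unused in the AnimGraph.
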AnimBlueprint /BRCosmetics/Characters/Player/Male/Large/Bodies/M_LRG_CrumbViolin/Meshes/M_LRG_CrumbViolin_AnimBP
target_skeleton=&/M_LRG_CrumbViolin_Skeleton
graphs [
 AnimGraph {nodes [
   n1 = Slot 'DefaultSlot'()
   n2 = Use cached pose 'preSimCache'()
   n3 = Get Is Playing Custom Emote() -> (Is Playing Custom Emote:double)
   n4 = "Constraint - Joystick_l_jnt (P)"(
    ComponentPose=Local To Component(LocalPose=Layered blend per bone(BasePose=@n2, BlendPoses_0=@n1, BlendWeights_0=@n3) {layerSetup [{branchFilters [{boneName=Joystick_l_jnt}, {boneName=Joystick_r_jnt}]}]})
    Alpha=1
    ConstraintWeights_0=1
   ) {
    alphaBoolBlend {blendOption=Linear}
    boneToModify {boneName=Joystick_l_jnt}
    constraintSetup [{targetBone {boneName=spine_01}}]
    constraintWeights [1]
   }
   n5 = "Constraint - Joystick_r_jnt (P)"(ComponentPose=@n4, Alpha=1, ConstraintWeights_0=1) {
    alphaBoolBlend {blendOption=Linear}
    boneToModify {boneName=Joystick_r_jnt}
    constraintSetup [{targetBone {boneName=spine_01}}]
    constraintWeights [1]
   }
   Output Pose(Result=Component To Local(ComponentPose=@n5)) {layerGroup=None}
   n6 = Control Rig(Source=Copy Pose From Mesh() {bUseAttachedParent=true}) {
    alphaBoolBlend {blendOption=Linear}
    controlRigAssetReference {blueprintRigClass=&/M_LRG_CrumbViolin_CtrlRig.M_LRG_CrumbViolin_CtrlRig_C}
   }
   Save cached pose 'CRCache'(Pose=@n6) {cacheName=CRCache}
   n7 = "Constraint - dyn_shoulderPad_l_01 (PP)"!(ComponentPose=Local To Component(LocalPose=Use cached pose 'CRCache'()), Alpha=1, ConstraintWeights_0=0.3, ConstraintWeights_1:float=0.7) -> (Pose) {
    alphaBoolBlend {blendOption=Linear}
    alphaScaleBiasClamp {bInterpResult=true, interpSpeedDecreasing=4, interpSpeedIncreasing=4}
    boneToModify {boneName=dyn_shoulderPad_l_01}
    constraintSetup [{targetBone {boneName=clavicle_l}}, {targetBone {boneName=upperarm_l}}]
    constraintWeights [1,1]
   }
   n8 = "Constraint - dyn_shoulderPad_r_01 (PP)"!(ComponentPose=@n7, Alpha=1, ConstraintWeights_0=0.3, ConstraintWeights_1:float=0.7) -> (Pose) {
    alphaBoolBlend {blendOption=Linear}
    alphaScaleBiasClamp {bInterpResult=true, interpSpeedDecreasing=4, interpSpeedIncreasing=4}
    boneToModify {boneName=dyn_shoulderPad_r_01}
    constraintSetup [{targetBone {boneName=clavicle_r}}, {targetBone {boneName=upperarm_r}}]
    constraintWeights [1,1]
   }
   n9 = "Transform (Modify) Bone - Bone: dyn_shoulderPad_l_01"(ComponentPose=@n8, Alpha=1, Translation="7,1,0", Rotation="0,0,0", Scale="1,1,1") {
    rotationMode=BMM_Additive
    rotationSpace=BCS_BoneSpace
    translationMode=BMM_Additive
    translationSpace=BCS_BoneSpace
    alphaBoolBlend {blendOption=Linear}
    alphaScaleBiasClamp {bInterpResult=true, interpSpeedDecreasing=4, interpSpeedIncreasing=4}
    boneToModify {boneName=dyn_shoulderPad_l_01}
    translation=<7 1>
   }
   n10 = "Transform (Modify) Bone - Bone: dyn_shoulderPad_r_01"(ComponentPose=@n9, Alpha=1, Translation="-7,-1,0", Rotation="0,0,0", Scale="1,1,1") {
    rotationMode=BMM_Additive
    rotationSpace=BCS_BoneSpace
    translationMode=BMM_Additive
    translationSpace=BCS_BoneSpace
    alphaBoolBlend {blendOption=Linear}
    alphaScaleBiasClamp {bInterpResult=true, interpSpeedDecreasing=4, interpSpeedIncreasing=4}
    boneToModify {boneName=dyn_shoulderPad_r_01}
    translation=<-7 -1>
   }
   Save cached pose 'preSimCache'(Pose=Component To Local(ComponentPose=@n10)) {cacheName=preSimCache}
  ]}
 EventGraph {nodes [
   Event BlueprintUpdateAnimation()
   > n1 = Sequence!(execute) -> (then_0, then_1, then_2:Exec, then_3:Exec)
   > Set Is Parachute opened Alpha(Is Parachute opened Alpha:double="To Float (Boolean)"(InBool=Get bIsParachuteOpened())) -> (Output_Get:double)
   n2 = Branch(
    execute=@n1.then_1
    Condition=BackpackItemDefinitionHasAnyMetaTag!(Tags:Array of Gameplay Tag Structures=Make Array("[0]"=(TagName="Cosmetics.MetaTags.PeacefulPoem")), bRequiresEquipped:bool=false) -> (ReturnValue:bool)
   )
   > Set PeacefulPoem Backbling Alpha(PeacefulPoem Backbling Alpha:bool=true) -> (Output_Get:bool)
   Set PeacefulPoem Backbling Alpha(execute=@n2.else, PeacefulPoem Backbling Alpha=false)
   n3 = GetAnimInstance(self=GetOwnerSkeletalMeshComponent())
   Set Is Playing Custom Emote(
    execute=@n1.then_2
    Is Playing Custom Emote:double=SelectFloat(A=1, B=0, bPickA=Montage_IsPlaying(self=@n3, Montage=/BRCosmetics/Animation/Game/MainPlayer/Emotes/CrumbViolin/CLM/Emote_CrumbViolin_CLM_M))
   ) -> (Output_Get:double)
   Branch(execute=@n1.then_3, Condition=Get bIsInFrontEnd())
   > Set Is Playing Custom Emote(Is Playing Custom Emote=GetCurveValue(self=GetOwnerFrontEndAnimInstance(), CurveName=Frontend_Idle_Bypass))
  ]}
]
variables [
 Is Parachute opened Alpha {type=number}
 Is Playing Custom Emote {type=number}
 PeacefulPoem Backbling Alpha {type=boolean}
]
