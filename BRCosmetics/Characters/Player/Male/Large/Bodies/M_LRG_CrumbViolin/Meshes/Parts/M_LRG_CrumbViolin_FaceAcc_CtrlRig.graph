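# Control Rig for the M_LRG_CrumbViolin character's face accessory that procedurally animates two tentacle beard chains and a spine adjustment.
# It applies a small rotation to each beard base bone, then runs per-item chain harmonics with pendulum and differing wave settings per side.
# A spherical pose reader on spine_05, relative to spine_01, then weights a corrective rotation of spine_01_adj, the parent of the face and joystick bones.
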ControlRigBlueprint /BRCosmetics/Characters/Player/Male/Large/Bodies/M_LRG_CrumbViolin/Meshes/Parts/M_LRG_CrumbViolin_FaceAcc_CtrlRig
graphs [
 RigVMFunctionLibrary {links [], nodes []}
 RigVMModel {nodes [
   RigUnit_BeginExecution = &RigUnit_BeginExecution()
   &RigUnit_ModifyTransforms(
    ItemToModify=((Item=(Type=Bone,Name="base_Tentacle_beard_l"),Transform=(Rotation=(X=0,Y=0.0523,Z=0,W=0.9986),Translation=(X=0,Y=0,Z=0),Scale3D=(X=1,Y=1,Z=1))))
    Weight=1
    WeightMinimum=0
    ExecutePin=@RigUnit_BeginExecution.ExecutePin
   )
   > &RigUnit_ModifyTransforms(
    ItemToModify=((Item=(Type=Bone,Name="base_Tentacle_beard_r"),Transform=(Rotation=(X=0,Y=0.0523,Z=0,W=0.9986),Translation=(X=0,Y=0,Z=0),Scale3D=(X=1,Y=1,Z=1))))
    Weight=1
    WeightMinimum=0
   )
   > /Script/ControlRig.RigUnit_ChainHarmonicsPerItem(
    ChainRoot=(Type=Bone,Name="base_Tentacle_beard_l")
    DrawWorldOffset=(Rotation=(X=0,Y=0,Z=0,W=1),Translation=(X=0,Y=0,Z=0),Scale3D=(X=1,Y=1,Z=1))
    Pendulum=(bEnabled=true,PendulumStiffness=2,PendulumGravity=(X=0,Y=0,Z=0),PendulumBlend=0.75,PendulumDrag=0.98,PendulumMinimum=0,PendulumMaximum=0.1,PendulumEase=Linear,UnwindAxis=(X=0,Y=1,Z=0),UnwindMinimum=0.2,UnwindMaximum=0.05)
    Reach=(bEnabled=false,ReachTarget=(X=0,Y=0,Z=0),ReachAxis=(X=1,Y=0,Z=0),ReachMinimum=0,ReachMaximum=0,ReachEase=Linear)
    Speed=(X=6,Y=7.2,Z=6.6)
    Wave=(bEnabled=True,WaveFrequency=(X=0,Y=0.8,Z=0.2),WaveAmplitude=(X=0.2,Y=0.5,Z=0.25),WaveOffset=(X=0,Y=0,Z=0.2),WaveNoise=(X=0,Y=0.2,Z=0.2),WaveMinimum=0,WaveMaximum=1,WaveEase=Linear)
    WaveCurve=(EditorCurveData=(Keys=((),(Time=1,Value=1))))
    bDrawDebug=True
   ) {is=struct RigVMUnitNode}
   > &RigUnit_ChainHarmonicsPerItem(
    ChainRoot=(Type=Bone,Name="base_Tentacle_beard_r")
    DrawWorldOffset=(Rotation=(X=0,Y=0,Z=0,W=1),Translation=(X=0,Y=0,Z=0),Scale3D=(X=1,Y=1,Z=1))
    Pendulum=(bEnabled=true,PendulumStiffness=2,PendulumGravity=(X=0,Y=0,Z=0),PendulumBlend=0.75,PendulumDrag=0.98,PendulumMinimum=0,PendulumMaximum=0.1,PendulumEase=Linear,UnwindAxis=(X=0,Y=1,Z=0),UnwindMinimum=0.2,UnwindMaximum=0.05)
    Reach=(bEnabled=false,ReachTarget=(X=0,Y=0,Z=0),ReachAxis=(X=1,Y=0,Z=0),ReachMinimum=0,ReachMaximum=0,ReachEase=Linear)
    Speed=(X=6,Y=7.2,Z=5.6)
    Wave=(bEnabled=True,WaveFrequency=(X=0,Y=-0.6,Z=0.2),WaveAmplitude=(X=0.5,Y=0.4,Z=-0.3),WaveOffset=(X=0,Y=2,Z=2),WaveNoise=(X=0,Y=0.2,Z=0.2),WaveMinimum=0,WaveMaximum=1,WaveEase=Linear)
    WaveCurve=(EditorCurveData=(Keys=((),(Time=1,Value=1))))
    bDrawDebug=True
   )
   > n1 = &RigUnit_SphericalPoseReader(
    ActiveRegionScaleFactors=(PositiveWidth=1,NegativeWidth=1,PositiveHeight=1,NegativeHeight=1)
    ActiveRegionSize=0.15
    Debug=(bDrawDebug=true,bDraw2D=false,bDrawLocalAxes=false,DebugScale=25,DebugSegments=20,DebugThickness=0.25)
    DriverAxis=(X=1,Y=0,Z=0)
    DriverItem=(Type=Bone,Name="spine_05")
    FalloffRegionScaleFactors=(PositiveWidth=1,NegativeWidth=1,PositiveHeight=1,NegativeHeight=1)
    FalloffSize=0.35
    OptionalParentItem=(Type=Bone,Name="spine_01")
    RotationOffset=(X=65,Y=-90,Z=0)
   )
   > &RigUnit_ModifyTransforms(
    ItemToModify=((Item=(Type=Bone,Name="spine_01_adj"),Transform=(Rotation=(X=0,Y=0,Z=0.1736,W=0.9848),Translation=(X=0,Y=0,Z=0),Scale3D=(X=1,Y=1,Z=1))))
    Weight=@n1.OutputParam
   )
  ]}
]
hierarchy [
 root {loc=<0 0 0>, rot=<0 0 0 1>}
 pelvis {parent=root, loc=<0 -0.7321 92.4218>, rot=<0.0127 0.707 -0.0127 -0.707>}
 spine_01 {parent=pelvis, loc=<5.6002 -0.5128 0>, rot=<0 0 0.034 -0.9994>}
 spine_02 {parent=spine_01, loc=<8.0756 0.0139 0>, rot=<0 0 0.0238 -0.9997>}
 spine_03 {parent=spine_02, loc=<7.0395 0 0>, rot=<0 0 0.0149 -0.9999>}
 spine_04 {parent=spine_03, loc=<8.1023 0 0>, rot=<0 0 -0.0132 -0.9999>}
 spine_05 {parent=spine_04, loc=<10.0141 0 0>, rot=<0 0 -0.0015 -1>}
 clavicle_l {parent=spine_05, loc=<2.5856 -0.2866 -4.3777>, rot=<0.0465 -0.7289 -0.009 -0.6829>}
 clavicle_r {parent=spine_05, loc=<2.5855 -0.2866 4.3777>, rot=<0.7289 0.0465 -0.6829 0.009>}
 deform_pec_aim_base_l {parent=spine_04, loc=<5.0227 13.6719 -2.6508>, rot=<0.0672 -0.5201 0.3228 -0.7879>}
 deform_pec_aim_l {parent=deform_pec_aim_base_l, loc=<0 0 0>, rot=<0 0 0 1>}
 deform_pec_base_l {parent=deform_pec_aim_l, loc=<16.3279 0 0>, rot=<0 0 0 1>}
 deform_pec_l {parent=deform_pec_base_l, loc=<-4.6529 0 0>, rot=<0 0 0 1>}
 deform_pec_aim_base_r {parent=spine_04, loc=<5.0225 13.6718 2.6508>, rot=<-0.5201 -0.0672 0.7879 0.3228>}
 deform_pec_aim_r {parent=deform_pec_aim_base_r, loc=<0 0 0>, rot=<0 0 0 1>}
 deform_pec_base_r {parent=deform_pec_aim_r, loc=<-16.3279 0 0>, rot=<0 0 0 -1>}
 deform_pec_r {parent=deform_pec_base_r, loc=<4.6529 0 0.0001>, rot=<0 0 0 -1>}
 spine_01_adj {parent=spine_01, loc=<0 0 0>, rot=<0 0 0 1>}
 Joystick_l_jnt {parent=spine_01_adj, loc=<-0.7127 12.5437 -7.8232>, rot=<0.0015 -0.0112 -0.1114 -0.9937>}
 JoystickTip_l_jnt {parent=Joystick_l_jnt, loc=<6.7812 0 0>, rot=<0 0 0 1>}
 Joystick_r_jnt {parent=spine_01_adj, loc=<-0.7127 12.5437 7.8232>, rot=<0.9937 0.1114 0.0112 -0.0015>}
 JoystickTip_r_jnt {parent=Joystick_r_jnt, loc=<6.7811 -0.0001 0>, rot=<0 0 0 1>}
 faceAttach {parent=spine_01_adj, loc=<7.9663 0.9572 0>, rot=<0.0114 -0.707 -0.0114 -0.707>}
 C_brow_mid {parent=faceAttach, loc=<0 14.1687 0.589>, rot=<0 0 0 -1>}
 C_jaw {parent=faceAttach, loc=<-0.0134 1.8335 -2.1243>, rot=<0 0 0 1>}
 C_lip_lower_mid {parent=C_jaw, loc=<0.0223 9.1825 -4.8948>, rot=<0 0 0 -1>}
 L_lip_lower_outer {parent=C_jaw, loc=<3.9024 8.4099 -4.8384>, rot=<0.013 0.0445 0.3057 -0.951>}
 R_lip_lower_outer {parent=C_jaw, loc=<-3.8755 8.4099 -4.8384>, rot=<0.013 -0.0445 -0.3057 -0.951>}
 teeth_lower {parent=C_jaw, loc=<0.0134 5.8521 -4.2998>, rot=<0 0 0 1>}
 base_Tentacle_beard_l {parent=C_jaw, loc=<2.0157 9.0074 -5.7393>, rot=<0.0521 -0.0478 -0.7352 -0.6741>}
 dyn_Tentacle_beard_01_l {parent=base_Tentacle_beard_l, loc=<0 0 0>, rot=<0 0 0 1>}
 dyn_Tentacle_beard_02_l {parent=dyn_Tentacle_beard_01_l, loc=<2.2708 0 0>, rot=<0 0 0.4249 -0.9052>}
 dyn_Tentacle_beard_03_l {parent=dyn_Tentacle_beard_02_l, loc=<1.2595 0 0>, rot=<0 0 -0.4622 -0.8868>}
 base_Tentacle_beard_r {parent=C_jaw, loc=<-2.8633 8.8432 -5.5777>, rot=<0.1298 -0.0775 -0.8488 -0.5067>}
 dyn_Tentacle_beard_01_r {parent=base_Tentacle_beard_r, loc=<0 0 0>, rot=<0 0 0 1>}
 dyn_Tentacle_beard_02_r {parent=dyn_Tentacle_beard_01_r, loc=<1.5648 0 0>, rot=<-0.0094 0.0957 0.0969 -0.9906>}
 dyn_Tentacle_beard_03_r {parent=dyn_Tentacle_beard_02_r, loc=<1.2323 0 0>, rot=<0 0 0.5647 -0.8253>}
 dyn_Tentacle_beard_04_r {parent=dyn_Tentacle_beard_03_r, loc=<1.2376 0 0>, rot=<0 0 0.0314 -0.9995>}
 C_lip_upper_mid {parent=faceAttach, loc=<0.0022 13.9048 -1.9081>, rot=<0.2225 0.0013 -0.0005 -0.9749>}
 C_nose_bridge {parent=faceAttach, loc=<0 13.7297 -0.8793>, rot=<-0.2218 0 0 -0.9751>}
 L_brow_mid {parent=faceAttach, loc=<3.514 13.143 2.4151>, rot=<-0.0001 -0.0042 0.2553 -0.9669>}
 L_brow_outer {parent=faceAttach, loc=<7.0021 10.5632 4.1779>, rot=<-0.0287 -0.04 0.4491 -0.8921>}
 L_cheek_inner {parent=faceAttach, loc=<9.5594 7.0431 -2.1062>, rot=<-0.0706 -0.2697 0.3184 -0.906>}
 L_eye {parent=faceAttach, loc=<3.3085 10.763 0.9644>, rot=<0 0 0.9998 0.0175>}
 L_eye_lid_lower_mid {parent=L_eye, loc=<0 0 0>, rot=<0.0329 0 0 -0.9995>}
 L_eye_lid_upper_mid {parent=L_eye, loc=<0 0 0>, rot=<0.0523 0 0 -0.9986>}
 L_lip_corner {parent=faceAttach, loc=<7.7356 9.5617 -1.934>, rot=<0.003 0.008 -0.0146 -0.9999>}
 L_lip_upper_outer {parent=faceAttach, loc=<4.5761 12.2216 -2.3372>, rot=<0.0359 -0.0728 0.2627 -0.9615>}
 R_brow_mid {parent=faceAttach, loc=<-3.514 13.143 2.4151>, rot=<-0.0001 0.0042 -0.2553 -0.9669>}
 R_brow_outer {parent=faceAttach, loc=<-7.0021 10.5632 4.1779>, rot=<-0.0287 0.04 -0.4491 -0.8921>}
 R_cheek_inner {parent=faceAttach, loc=<-9.5594 7.0431 -2.1062>, rot=<-0.0706 0.2697 -0.3184 -0.906>}
 R_eye {parent=faceAttach, loc=<-3.3085 10.763 0.9643>, rot=<0 0 0.9998 -0.0175>}
 R_eye_lid_lower_mid {parent=R_eye, loc=<0 -0.0002 0.0001>, rot=<0.0329 0 0 -0.9995>}
 R_eye_lid_upper_mid {parent=R_eye, loc=<0 0 0>, rot=<0.0523 0 0 -0.9986>}
 R_lip_corner {parent=faceAttach, loc=<-7.7356 9.5617 -1.934>, rot=<0.003 -0.008 0.0146 -0.9999>}
 R_lip_upper_outer {parent=faceAttach, loc=<-4.5761 12.2216 -2.3372>, rot=<0.0359 0.0728 -0.2627 -0.9615>}
 thigh_l {parent=pelvis, loc=<-1.8716 -0.0109 -11.9936>, rot=<0.0012 0.0572 0.0155 -0.9982>}
 thigh_twist_01_l {parent=thigh_l, loc=<-21.6303 0 0>, rot=<0.0013 0 0 -1>}
 deform_groin_base_l {parent=thigh_l, loc=<-2.1908 8.5661 -1.7442>, rot=<0 0 0 -1>}
 deform_groin_l {parent=deform_groin_base_l, loc=<0 0 0>, rot=<0 0 0 1>}
 thigh_r {parent=pelvis, loc=<-1.8716 -0.0109 11.9936>, rot=<0.0572 -0.0012 0.9982 0.0155>}
 thigh_twist_01_r {parent=thigh_r, loc=<21.6303 0 0>, rot=<0.0013 0 0 -1>}
 deform_groin_base_r {parent=thigh_r, loc=<2.1908 -8.5661 1.7442>, rot=<0 0 0 -1>}
 deform_groin_r {parent=deform_groin_base_r, loc=<0 0 0>, rot=<0 0 0 -1>}
]
