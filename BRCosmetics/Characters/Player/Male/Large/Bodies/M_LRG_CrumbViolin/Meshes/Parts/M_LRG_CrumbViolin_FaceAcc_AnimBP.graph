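# Animation Blueprint for the CrumbViolin large male face accessory, posing it from the character's meshes.
# It layers joystick and beard-tentacle bones from the body mesh over the parent pose, then applies a Control Rig and fixed faceAttach and jaw offsets.
# The cached DefaultSlot pose overrides the beard branch during the CrumbViolin emote and faceAttach by front-end idle curve weight.
# The event graph also toggles a material hide parameter when a PeacefulPoem-tagged backpack is detected.
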AnimBlueprint /BRCosmetics/Characters/Player/Male/Large/Bodies/M_LRG_CrumbViolin/Meshes/Parts/M_LRG_CrumbViolin_FaceAcc_AnimBP
target_skeleton=&/M_LRG_CrumbViolin_FaceAcc_Skeleton
graphs [
 AnimGraph {nodes [
   n1 = Layered blend per bone(
    BasePose=Copy Pose From Mesh() {bCopyCurves=true, bUseAttachedParent=true}
    BlendPoses_0=Copy Pose From Mesh(SourceMeshComponent=Get body Mesh() -> (body Mesh:Skeletal Mesh Component Object Reference))
   ) {layerSetup [{branchFilters [{boneName=Joystick_l_jnt}, {boneName=Joystick_r_jnt}, {boneName=dyn_Tentacle_beard_04_r}]}]}
   n2 = Control Rig(Source=@n1) {
    alphaBoolBlend {blendOption=Linear}
    controlRigAssetReference {blueprintRigClass=&/M_LRG_CrumbViolin_FaceAcc_CtrlRig.M_LRG_CrumbViolin_FaceAcc_CtrlRig_C}
   }
   n3 = "Transform (Modify) Bone - Bone: faceAttach"(ComponentPose=Local To Component(LocalPose=@n2), Alpha=1, Translation="0,0,5", Rotation="0,0,0", Scale="1,1,1") {
    rotationMode=BMM_Additive
    rotationSpace=BCS_BoneSpace
    translationMode=BMM_Additive
    translationSpace=BCS_BoneSpace
    alphaBoolBlend {blendOption=Linear}
    boneToModify {boneName=faceAttach}
    translation {z=5}
   }
   n4 = "Transform (Modify) Bone - Bone: C_jaw"(ComponentPose=@n3, Alpha=1, Translation="0,0,0", Rotation="0,0,-18", Scale="1,1,1") {
    rotationMode=BMM_Additive
    rotationSpace=BCS_BoneSpace
    translationMode=BMM_Additive
    translationSpace=BCS_BoneSpace
    alphaBoolBlend {blendOption=Linear}
    boneToModify {boneName=C_jaw}
    rotation {roll=-18}
   }
   n5 = Component To Local(ComponentPose=@n4)
   n6 = Use cached pose 'MontageReference'()
   n7 = Get Is Playing Custom Emote() -> (Is Playing Custom Emote:double)
   Output Pose(Result=Layered blend per bone(
     BasePose=Layered blend per bone(BasePose=@n5, BlendPoses_0=@n6, BlendWeights_0=@n7) {layerSetup [{branchFilters [{boneName=dyn_Tentacle_beard_04_r}]}]}
     BlendPoses_0=Use cached pose 'MontageReference'()
     BlendWeights_0=Get FrontEnd Custom Lobby Idle() -> (FrontEnd Custom Lobby Idle:float)
    ) {layerSetup [{branchFilters [{boneName=faceAttach}]}]}) {layerGroup=None}
   Save cached pose 'MontageReference'(Pose=Slot 'DefaultSlot'()) {cacheName=MontageReference}
  ]}
 EventGraph {nodes [
   Event BlueprintUpdateAnimation()
   > n1 = Sequence!(execute) -> (then_0, then_1, then_2:Exec)
   n2 = Get Dynamic Material Instance(execute=[@n6.then_1, @n1.then_0], SkeletalMeshComponent=GetPartSkeletalMeshComponent())
   > n3 = Branch(Condition=BackpackItemDefinitionHasAnyMetaTag!(Tags:Array of Gameplay Tag Structures=Make Array("[0]"=(TagName="Cosmetics.MetaTags.PeacefulPoem")), bRequiresEquipped:bool=false) -> (ReturnValue:bool))
   > Set PeacefulPoem Backbling Alpha(PeacefulPoem Backbling Alpha:bool=true) -> (Output_Get:bool)
   > SetScalarParameterValue(self=@n2, ParameterName=Hide Element 02, Value=0)
   Set PeacefulPoem Backbling Alpha(execute=@n3.else, PeacefulPoem Backbling Alpha=false)
   > SetScalarParameterValue(self=@n2, ParameterName=Hide Element 02, Value=1)
   n4 = GetAnimInstance(self=GetOwnerSkeletalMeshComponent())
   Set Is Playing Custom Emote(
    execute=@n1.then_1
    Is Playing Custom Emote:double=SelectFloat(A=1, B=0, bPickA=Montage_IsPlaying(self=@n4, Montage=/BRCosmetics/Animation/Game/MainPlayer/Emotes/CrumbViolin/CLM/Emote_CrumbViolin_CLM_M))
   ) -> (Output_Get:double)
   Branch(execute=@n1.then_2, Condition=Get bIsInFrontEnd())
   > Set FrontEnd Custom Lobby Idle(FrontEnd Custom Lobby Idle:float=GetCurveValue(self=GetOwnerFrontEndAnimInstance(), CurveName=Frontend_Idle_Bypass)) -> (Output_Get:float)
   n5 = Event OnCharacterCustomizationLoaded()
   > n6 = Sequence()
   n7 = Get Array Item with EFort Custom Part Type(Array=@n5.CustomMeshParts)
   > Set body Mesh(body Mesh:Skeletal Mesh Component Object Reference=@n7) -> (Output_Get:Skeletal Mesh Component Object Reference)
   Event BlueprintInitializeAnimation()
   > Set body Mesh(body Mesh=@n7)
   GetOwnerSkeletalMeshComponent()
   GetOwnerSkeletalMeshComponent()
  ]}
]
variables [
 body Mesh {type=SkeletalMeshComponent}
 FrontEnd Custom Lobby Idle {type=number}
 Is Playing Custom Emote {type=number}
 PeacefulPoem Backbling Alpha {type=boolean}
]
